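; Parent Blueprint for commerce food truck props, handling an optional truck topper and its bracket.
; Its construction script first destroys the topper or bracket component if it has no static mesh assigned.
; It then, when topper rotation is enabled, applies an animated material and turns on World Position Offset with a disable distance, so rotation comes from the material rather than Tick.
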(event EventBeginPlay)

(event Collision|EventActorBeginOverlap
  (params (OtherActor :type Actor%20Object%20Reference)))

(event EventTick
  (params (DeltaSeconds :type Float%20%28single-precision%29)))

(fn SetTopperRotation
  (bind _trucktopper (Variables|Default|GetTruckTopper))
  (if (Variables|Default|GetEnableTopperRotation)
    (Utilities|IsValid _trucktopper
      (:Is%20Valid
        (Rendering|Material|SetMaterial _trucktopper 0 (Variables|Default|GetAnimatedTopperMaterial))
        (Variables|/Script/Engine.StaticMeshComponent|SetbEvaluateWorldPositionOffset true _trucktopper)
        (Rendering|LOD|SetWorldPositionOffsetDisableDistance _trucktopper 6000 :!class /Script/Engine.StaticMeshComponent)))))

(fn ConstructionScript
  (|CallParentFunction)
  (CallFunction|ValidateTopperMeshes)
  (CallFunction|SetTopperRotation))

(fn ValidateTopperMeshes
  (bind _bracket (Variables|Default|GetBracket))
  (Utilities|IsValid (Variables|/Script/Engine.StaticMeshComponent|GetStaticMesh (Variables|Default|GetTruckTopper))
    (:Is%20Valid
      _)
    (:Is%20Not%20Valid
      (Components|DestroyComponent (Variables|Default|GetTruckTopper))))
  (Utilities|IsValid (Variables|/Script/Engine.StaticMeshComponent|GetStaticMesh _bracket)
    (:Is%20Not%20Valid
      (Components|DestroyComponent _bracket))))
